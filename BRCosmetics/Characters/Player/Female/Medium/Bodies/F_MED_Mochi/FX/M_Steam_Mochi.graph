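# Unlit, two-sided translucent steam material for Niagara mesh particles on the Mochi female player cosmetic.
# A panned smoke trail texture and a soft alpha, with UV offsets for noise-style detail distortion, build opacity, tinted and faded by particle color.
# Dynamic parameters offset time per particle, set tapered SplineThicken width and drive a rolling sphere mask that gates both world position offset and opacity.
# Depth fade softens intersections, except on mobile.
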Material /BRCosmetics/Characters/Player/Female/Medium/Bodies/F_MED_Mochi/FX/M_Steam_Mochi
graph {
 properties {
  bUsedWithNiagaraMeshParticles=true
  blendMode=BLEND_Translucent
  pixelDepthOffsetMode=PDOM_Legacy
  shadingModel=MSM_Unlit
  twoSided=true
  usedShadingModels=Unlit
 }
 nodes [
  n1 = ParticleColor()
  MaterialFunctionCall(Emissive=@n1.RGB) {materialFunction=&MF_FX_EmissiveAdjustments}
  n2 = DynamicParameter() {paramNames [SplineThickness, Random, RollingMaskPos, RollingMaskWidth]}
  n3 = SphereMask(A=ComponentMask(None=NamedRerouteUsage()) {r=true}, B=@n2.RollingMaskPos, Radius=@n2.RollingMaskPos) {hardnessPercent=0}
  n4 = DynamicParameter() {paramNames [SplineThickness, Random, RollingMaskPos, RollingMaskWidth]}
  n5 = ScalarParameter() {defaultValue=5, parameterName=TaperingMultiplier}
  Multiply(
   A=Step(Y=@n3)
   B=MaterialFunctionCall(WidthBase=Multiply(A=@n4.SplineThickness, B=@n5), WidthTip=@n4.SplineThickness) {materialFunction=&SplineThicken}.WorldPositionOffset
  )
  NamedRerouteDeclaration(None=MaterialFunctionCall(XY=TextureCoordinate()) {materialFunction=&Swizzle}.YX) {name=MainUV, nodeColor=<rgba 1 0 0.75 1>}
  NamedRerouteDeclaration(None=Add(A=Time(), B=@n2.Random)) {name=PerParticleTime, nodeColor=<rgba 0 1 0.3438 1>}
  n6 = VectorParameter() {parameterName=SmokeTrailParams, channelNames {a=SpeedY, b=SpeedX, g=ScaleY, r=ScaleX}, defaultValue=<rgba 0.15 1 -0.5 0>}
  n7 = Multiply(A=AppendVector(A=@n6.ScaleX, B=@n6.ScaleX), B=NamedRerouteUsage())
  n8 = NamedRerouteUsage()
  n9 = AppendVector(A=@n6.SpeedX, B=@n6.SpeedX)
  n10 = NamedRerouteUsage()
  n11 = VectorParameter() {
   parameterName=DetailDistortionWeights
   channelNames {a=DistortAlphaX, b=DistortAlphaX, g=DistortTrailY, r=DistortTrailX}
   defaultValue=<rgba 0.1 0.1 0.1 0>
  }
  n12 = AppendVector(A=@n11.DistortTrailX, B=@n11.DistortTrailX)
  n13 = TextureSampleParameter2D(UVs=Add(A=Panner(Coordinate=@n7, Time=@n8, Speed=@n9), B=Multiply(A=@n10, B=@n12))) {
   parameterName=T_SmokeTrailTex
   samplerSource=SSM_Wrap_WorldGroupSettings
   texture=/Game/Effects/Fort_Effects/Textures/Smoke/T_Smoke_Trail_Soft_2
  }
  n14 = AppendVector(A=@n11.DistortAlphaX, B=@n11.DistortAlphaX)
  n15 = TextureSampleParameter2D(UVs=Add(A=TextureCoordinate(), B=Multiply(A=@n14, B=@n10))) {parameterName=T_Alpha, texture=/Game/Effects/Fort_Effects/Textures/T_Smoke_Soft}
  NamedRerouteDeclaration(None=Multiply(A=Multiply(A=@n13.G, B=@n15.G), B=NamedRerouteUsage())) {name=TexturedOpacity, nodeColor=<rgba 1 0.5625 0 1>}
  n16 = VectorParameter() {parameterName=DetailTexParams, channelNames {a=SpeedY, b=SpeedX, g=ScaleY, r=ScaleX}, defaultValue=<rgba 0.2 0.2 -0.5 0>}
  n17 = AppendVector(A=@n16.ScaleX, B=@n16.ScaleX)
  n18 = NamedRerouteUsage()
  n19 = TextureSampleParameter2D(UVs=Panner(Coordinate=Multiply(A=@n17, B=@n18), Time=NamedRerouteUsage(), Speed=AppendVector(A=@n16.SpeedX, B=@n16.SpeedX))) {parameterName=T_DetailTex, samplerSource=SSM_Wrap_WorldGroupSettings, texture=/Game/Global/Textures/Default/Noise/T_FireNoise_02}
  n20 = NamedRerouteDeclaration(None=@n19.R) {name=DetailSrc, nodeColor=<rgba 1 0.5859 0 1>}
  NamedRerouteDeclaration(None=Subtract(A=Multiply(A=@n20) {constB=2})) {name=DetailExpanded, nodeColor=<rgba 0.4766 0 1 1>}
  ParticleRandom()
  n21 = Multiply(A=Multiply(A=@n1.A, B=NamedRerouteUsage()), B=@n3)
  ShadingPathSwitch(Default=Multiply(A=@n21, B=DepthFade() {fadeDistanceDefault=15}), Mobile=@n21)
  ShadingPathSwitch(Default=LinearInterpolate(B=LinearInterpolate(Alpha=@n1.A) {constA=0.98}, Alpha=@n13.G) {constA=1, constB=0.98})
 ]
}
