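# Material function for player faces that shows swappable eyebrow, eye and mouth pose elements tagged by vertex color, driven by pose curve inputs.
# Per-feature channel masks (default green, red, blue) feed MF_MakeVertexColorElementVisible for each pose, merged by chained lerps into three WPO outputs.
# A red/blue vertex-color mask minus three summed signals gates the main output, apparently collapsing hidden vertices to the object position via world position offset.
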MaterialFunction /BRCosmetics/Characters/Player/Common/MaterialFunctions/VertexColorFaces/MF_VertexColorFaces
graph {nodes [
  n1 = VertexColor()
  n2 = Max(A=@n1.R, B=@n1.R)
  n3 = Ceil(None=Subtract(A=Max(A=@n2, B=@n1.B)) {constB=0.05})
  n4 = Add(A=Add(A=NamedRerouteUsage(), B=NamedRerouteUsage()), B=NamedRerouteUsage())
  n5 = Saturate(None=Round(None=Subtract(A=@n3, B=@n4)))
  FunctionOutput(None=Multiply(A=NamedRerouteUsage(), B=@n5))
  FunctionOutput(None=NamedRerouteUsage()) {outputName=Eyebrow WPO, sortPriority=3}
  FunctionOutput(None=NamedRerouteUsage()) {outputName=Eye WPO, sortPriority=2}
  FunctionOutput(None=NamedRerouteUsage()) {outputName=Mouth WPO, sortPriority=1}
  LinearInterpolate(Alpha=@n5)
  n6 = VertexColor()
  NamedRerouteDeclaration(None=ChannelMaskParameter(None=@n6) {maskChannel=Green, parameterName=Eyebrow Channel, defaultValue=<rgba 0 1 0 0>}.R) {name=Eyebrow_Vert_Color, nodeColor=<rgba 0.6551 0.0306 1 1>}
  NamedRerouteDeclaration(None=ChannelMaskParameter(None=@n6) {parameterName=Eye Channel}.R) {name=Eye_Vert_Color, nodeColor=<rgba 0 1 0.9531 1>}
  n7 = ScalarParameter() {defaultValue=0.1, group=Eyebrow Pose A, parameterName=Eyebrow_Pose_A_Vertex_Color_Value, sortPriority=3}
  n8 = ScalarParameter() {defaultValue=0.2, group=Eyebrow Pose A, parameterName=Eyebrow_Pose_A_Threshold, sortPriority=2}
  n9 = FunctionInput(Preview=ScalarParameter() {group=Eyebrow Pose A, parameterName=R_brow_up_pose, sortPriority=1}) {bUsePreviewValueAsDefault=true, inputName=Eyebrow Pose A Curve, inputType=FunctionInput_Scalar, sortPriority=1}
  n10 = NamedRerouteUsage()
  n11 = MaterialFunctionCall!(
   Vertex Color Value to Hide=ScalarParameter() {defaultValue=0.1, group=Eyebrow Pose B, parameterName=Eyebrow_Pose_B_Vertex_Color_Value}
   Threshold=ScalarParameter() {defaultValue=0.2, group=Eyebrow Pose B, parameterName=Eyebrow_Pose_B_Threshold}
   Alpha=FunctionInput(Preview=ScalarParameter() {group=Eyebrow Pose B, parameterName=C_glabella_down_pose}) {bUsePreviewValueAsDefault=true, inputName=Eyebrow Pose B Curve, inputType=FunctionInput_Scalar, sortPriority=2}
   Vertex Color Channel=NamedRerouteUsage()
  ) {materialFunction=/BRCosmetics/Characters/Player/Female/Medium/Bodies/F_MED_Mochi/Materials/MF_MakeVertexColorElementVisible}
  n12 = MaterialFunctionCall(
   Vertex Color Value to Hide=ScalarParameter() {defaultValue=0.3, group=Eyebrow Pose C, parameterName=Eyebrow_Pose_C_Vertex_Color_Value}
   Threshold=ScalarParameter() {defaultValue=0.2, group=Eyebrow Pose C, parameterName=Eyebrow_Pose_C_Threshold}
   Alpha=FunctionInput(Preview=ScalarParameter() {group=Eyebrow Pose C, parameterName=C_glabella_up_pose}) {bUsePreviewValueAsDefault=true, inputName=Eyebrow Pose C Curve, inputType=FunctionInput_Scalar, sortPriority=3}
   Vertex Color Channel=NamedRerouteUsage()
  ) {materialFunction=&MF_MakeVertexColorElementVisible}
  n13 = LinearInterpolate(
   A=LinearInterpolate(
    A=MaterialFunctionCall(Vertex Color Value to Hide=@n7, Threshold=@n8, Alpha=@n9, Vertex Color Channel=@n10) {materialFunction=&MF_MakeVertexColorElementVisible}.Result
    B=@n11.Result
    Alpha=@n11.Result
   )
   B=@n12.Result
   Alpha=@n12.Result
  )
  n14 = MaterialFunctionCall(
   Vertex Color Value to Hide=ScalarParameter() {defaultValue=0.3, group=Eyebrow Pose D, parameterName=Eyebrow_Pose_D_Vertex_Color_Value}
   Threshold=ScalarParameter() {defaultValue=0.2, group=Eyebrow Pose D, parameterName=Eyebrow_Pose_D_Threshold}
   Alpha=FunctionInput(Preview=ScalarParameter() {group=Eyebrow Pose D, parameterName=C_glabella_up_pose}) {bUsePreviewValueAsDefault=true, inputName=Eyebrow Pose D Curve, inputType=FunctionInput_Scalar, sortPriority=4}
   Vertex Color Channel=NamedRerouteUsage()
  ) {materialFunction=&MF_MakeVertexColorElementVisible}
  n15 = MaterialFunctionCall(
   Vertex Color Value to Hide=ScalarParameter() {defaultValue=0.1, group=Eyebrow Pose E, parameterName=Eyebrow_Pose_E_Vertex_Color_Value}
   Threshold=ScalarParameter() {defaultValue=0.2, group=Eyebrow Pose E, parameterName=Eyebrow_Pose_E_Threshold}
   Alpha=FunctionInput(Preview=ScalarParameter() {group=Eyebrow Pose E, parameterName=C_glabella_down_pose}) {bUsePreviewValueAsDefault=true, inputName=Eyebrow Pose E Curve, inputType=FunctionInput_Scalar, sortPriority=5}
   Vertex Color Channel=NamedRerouteUsage()
  ) {materialFunction=&MF_MakeVertexColorElementVisible}
  NamedRerouteDeclaration(None=LinearInterpolate(A=LinearInterpolate(A=@n13, B=@n14.Result, Alpha=@n14.Result), B=@n15.Result, Alpha=@n15.Result)) {name=Eyebrow_WPO, nodeColor=<rgba 0 0.8125 1 1>}
  n16 = MaterialFunctionCall(
   Vertex Color Value to Hide=ScalarParameter() {defaultValue=0.1, group=Eye Pose A, parameterName=Eye_Pose_A_Vertex_Color_Value}
   Threshold=ScalarParameter() {defaultValue=0.2, group=Eye Pose A, parameterName=Eye_Pose_A_Threshold}
   Alpha=FunctionInput(Preview=ScalarParameter() {group=Eye Pose A, parameterName=R_blink_pose}) {bUsePreviewValueAsDefault=true, inputName=Eye Pose A Curve, inputType=FunctionInput_Scalar, sortPriority=7}
   Vertex Color Channel=NamedRerouteUsage()
  ) {materialFunction=&MF_MakeVertexColorElementVisible}
  n17 = MaterialFunctionCall(
   Vertex Color Value to Hide=ScalarParameter() {defaultValue=0.2, group=Eye Pose B, parameterName=Eye_Pose_B_Vertex_Color_Value}
   Threshold=ScalarParameter() {defaultValue=0.2, group=Eye Pose B, parameterName=Eye_Pose_B_Threshold}
   Alpha=FunctionInput() {bUsePreviewValueAsDefault=true, inputName=Eye Pose B Curve, inputType=FunctionInput_Scalar, sortPriority=8}
   Vertex Color Channel=NamedRerouteUsage()
  ) {materialFunction=&MF_MakeVertexColorElementVisible}
  n18 = MaterialFunctionCall(
   Vertex Color Value to Hide=ScalarParameter() {defaultValue=0.5, group=Eye Pose C, parameterName=Eye_Pose_C_Vertex_Color_Value}
   Threshold=ScalarParameter() {defaultValue=0.2, group=Eye Pose C, parameterName=Eye_Pose_C_Threshold}
   Alpha=FunctionInput(Preview=ScalarParameter() {group=Eye Pose C, parameterName=C_glabella_up_pose}) {bUsePreviewValueAsDefault=true, inputName=Eye Pose C Curve, inputType=FunctionInput_Scalar, sortPriority=9}
   Vertex Color Channel=NamedRerouteUsage()
  ) {materialFunction=&MF_MakeVertexColorElementVisible}
  NamedRerouteDeclaration(None=LinearInterpolate(A=LinearInterpolate(A=@n16.Result, B=@n17.Result, Alpha=@n17.Result), B=@n18.Result, Alpha=@n18.Result)) {name=Eye_WPO, nodeColor=<rgba 0 0.8125 1 1>}
  NamedRerouteDeclaration(None=ChannelMaskParameter(None=@n6) {maskChannel=Blue, parameterName=Mouth Channel, defaultValue=<rgba 0 0 1 0>}.R) {name=Mouth_Vert_Color, nodeColor=<rgba 0 1 0.9531 1>}
  n19 = WorldPosition()
  n20 = ObjectPositionWS()
  NamedRerouteDeclaration(None=Multiply(A=Subtract(A=@n19.XYZ, B=@n20)) {constB=-1}) {name=Vertex Color Hide, nodeColor=<rgba 0.6875 0 1 1>}
  n21 = MaterialFunctionCall(
   Vertex Color Value to Hide=ScalarParameter() {defaultValue=0.2, group=Mouth Pose A, parameterName=Mouth_Pose_A_Vertex_Color_Value, sortPriority=1}
   Threshold=ScalarParameter() {defaultValue=0.2, group=Mouth Pose A, parameterName=Mouth_Pose_A_Threshold, sortPriority=1}
   Alpha=FunctionInput() {bUsePreviewValueAsDefault=true, inputName=Mouth Pose A Curve, inputType=FunctionInput_Scalar, sortPriority=12}
   Vertex Color Channel=NamedRerouteUsage()
  ) {materialFunction=&MF_MakeVertexColorElementVisible}
  n22 = MaterialFunctionCall(
   Vertex Color Value to Hide=ScalarParameter() {defaultValue=0.2, group=Mouth Pose B, parameterName=Mouth_Pose_B_Vertex_Color_Value}
   Threshold=ScalarParameter() {defaultValue=0.33, group=Mouth Pose B, parameterName=Mouth_Pose_B_Threshold}
   Alpha=FunctionInput() {bUsePreviewValueAsDefault=true, inputName=Mouth Pose B Curve, inputType=FunctionInput_Scalar, sortPriority=13}
   Vertex Color Channel=NamedRerouteUsage()
  ) {materialFunction=&MF_MakeVertexColorElementVisible}
  n23 = MaterialFunctionCall(
   Vertex Color Value to Hide=ScalarParameter() {defaultValue=0.5, group=Mouth Pose C, parameterName=Mouth_Pose_C_Vertex_Color_Value}
   Threshold=ScalarParameter() {defaultValue=0.2, group=Mouth Pose C, parameterName=Mouth_Pose_C_Threshold}
   Alpha=FunctionInput() {bUsePreviewValueAsDefault=true, inputName=Mouth Pose C Curve, inputType=FunctionInput_Scalar, sortPriority=14}
   Vertex Color Channel=NamedRerouteUsage()
  ) {materialFunction=&MF_MakeVertexColorElementVisible}
  n24 = MaterialFunctionCall(
   Vertex Color Value to Hide=ScalarParameter() {defaultValue=0.5, group=Mouth Pose D, parameterName=Mouth_Pose_D_Vertex_Color_Value}
   Threshold=ScalarParameter() {defaultValue=0.2, group=Mouth Pose D, parameterName=Mouth_Pose_D_Threshold}
   Alpha=FunctionInput() {bUsePreviewValueAsDefault=true, inputName=Mouth Pose D Curve, inputType=FunctionInput_Scalar, sortPriority=15}
   Vertex Color Channel=NamedRerouteUsage()
  ) {materialFunction=&MF_MakeVertexColorElementVisible}
  n25 = LinearInterpolate(
   A=LinearInterpolate(A=LinearInterpolate(A=@n21.Result, B=@n22.Result, Alpha=@n22.Result), B=@n23.Result, Alpha=@n23.Result)
   B=@n24.Result
   Alpha=@n24.Result
  )
  n26 = MaterialFunctionCall(
   Vertex Color Value to Hide=ScalarParameter() {defaultValue=0.5, group=Mouth Pose E, parameterName=Mouth_Pose_E_Vertex_Color_Value}
   Threshold=ScalarParameter() {defaultValue=0.2, group=Mouth Pose E, parameterName=Mouth_Pose_E_Threshold}
   Alpha=FunctionInput() {bUsePreviewValueAsDefault=true, inputName=Mouth Pose E Curve, inputType=FunctionInput_Scalar, sortPriority=15}
   Vertex Color Channel=NamedRerouteUsage()
  ) {materialFunction=&MF_MakeVertexColorElementVisible}
  n27 = MaterialFunctionCall(
   Vertex Color Value to Hide=ScalarParameter() {defaultValue=0.5, group=Mouth Pose F, parameterName=Mouth_Pose_F_Vertex_Color_Value}
   Threshold=ScalarParameter() {defaultValue=0.2, group=Mouth Pose F, parameterName=Mouth_Pose_F_Threshold}
   Alpha=FunctionInput() {bUsePreviewValueAsDefault=true, inputName=Mouth Pose F Curve, inputType=FunctionInput_Scalar, sortPriority=15}
   Vertex Color Channel=NamedRerouteUsage()
  ) {materialFunction=&MF_MakeVertexColorElementVisible}
  n28 = MaterialFunctionCall(
   Vertex Color Value to Hide=ScalarParameter() {defaultValue=0.5, group=Mouth Pose G, parameterName=Mouth_Pose_G_Vertex_Color_Value}
   Threshold=ScalarParameter() {defaultValue=0.2, group=Mouth Pose G, parameterName=Mouth_Pose_G_Threshold}
   Alpha=FunctionInput() {bUsePreviewValueAsDefault=true, inputName=Mouth Pose G Curve, inputType=FunctionInput_Scalar, sortPriority=15}
   Vertex Color Channel=NamedRerouteUsage()
  ) {materialFunction=&MF_MakeVertexColorElementVisible}
  n29 = LinearInterpolate(
   A=LinearInterpolate(A=LinearInterpolate(A=@n25, B=@n26.Result, Alpha=@n26.Result), B=@n27.Result, Alpha=@n27.Result)
   B=@n28.Result
   Alpha=@n28.Result
  )
  n30 = MaterialFunctionCall(
   Vertex Color Value to Hide=ScalarParameter() {defaultValue=0.5, group=Mouth Pose H, parameterName=Mouth_Pose_H_Vertex_Color_Value}
   Threshold=ScalarParameter() {defaultValue=0.2, group=Mouth Pose H, parameterName=Mouth_Pose_H_Threshold}
   Alpha=FunctionInput() {bUsePreviewValueAsDefault=true, inputName=Mouth Pose H Curve, inputType=FunctionInput_Scalar, sortPriority=33}
   Vertex Color Channel=NamedRerouteUsage()
  ) {materialFunction=&MF_MakeVertexColorElementVisible}
  n31 = MaterialFunctionCall(
   Vertex Color Value to Hide=ScalarParameter() {defaultValue=0.5, group=Mouth Pose H, parameterName=Mouth_Pose_I_Vertex_Color_Value}
   Threshold=ScalarParameter() {defaultValue=0.2, group=Mouth Pose H, parameterName=Mouth_Pose_I_Threshold}
   Alpha=FunctionInput() {bUsePreviewValueAsDefault=true, inputName=Mouth Pose I Curve, inputType=FunctionInput_Scalar, sortPriority=34}
   Vertex Color Channel=NamedRerouteUsage()
  ) {materialFunction=&MF_MakeVertexColorElementVisible}
  NamedRerouteDeclaration(None=LinearInterpolate(A=LinearInterpolate(A=@n29, B=@n30.Result, Alpha=@n30.Result), B=@n31.Result, Alpha=@n31.Result)) {name=Mouth WPO, nodeColor=<rgba 0 1 0.25 1>}
 ]}
