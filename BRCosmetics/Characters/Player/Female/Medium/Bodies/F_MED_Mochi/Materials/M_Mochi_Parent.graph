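# Parent material for the F_MED_Mochi character, animating vertex-color face poses through World Position Offset.
# With IsMochiBody on, time-based waves sphere-masked against a random texture value, sampled by pre-skinned position, drive staggered blinks and smiles, zeroing several other poses.
# With it off, facial pose parameters drive the shapes.
# SkinFX-masked cloth fuzz and silk, then emissive fresnel, layer between the shared character PreFX and PostFX functions.
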Material /BRCosmetics/Characters/Player/Female/Medium/Bodies/F_MED_Mochi/Materials/M_Mochi_Parent
graph {
 properties {
  bUseMaterialAttributes=true
  bUsedWithMeshDeformer=true
  bUsedWithMorphTargets=true
  bUsedWithSkeletalMesh=true
  pixelDepthOffsetMode=PDOM_Legacy
  usedShadingModels=Default Lit
 }
 nodes [
  n1 = MaterialFunctionCall() {materialFunction=&BaseFN_CharacterFunction_PreFX}
  n2 = TextureSampleParameter2D() {parameterName=SkinFX, samplerType=SAMPLERTYPE_Masks, texture=&T_BlackMask}
  n3 = ChannelMaskParameter(None=@n2.RGB) {maskChannel=Green, parameterName=Silk Channel, defaultValue=<rgba 0 1 0 0>}
  n4 = Constant() {r=1}
  n5 = Multiply(A=Time(), B=ScalarParameter() {defaultValue=0.1235, parameterName=Speed})
  n6 = Sine(None=@n5)
  n7 = Frac(None=Multiply(A=@n5) {constB=0.7774})
  n8 = Saturate(None=Add(A=ConstantBiasScale(None=@n6), B=ConstantBiasScale(None=@n7) {bias=-0.5}))
  n9 = NamedRerouteUsage()
  n10 = ScalarParameter() {defaultValue=0.02, parameterName=Blink}
  n11 = ScalarParameter() {group=Eye Pose A, parameterName=R_blink_pose}
  n12 = LinearInterpolate(A=@n10, B=Multiply(A=@n10) {constB=4}, Alpha=@n11)
  n13 = StaticSwitch(
   True=Clamp(None=SphereMask(A=@n8, B=@n9, Radius=@n12) {attenuationRadius=0.3})
   False=@n11
   Value=StaticBoolParameter() {defaultValue=true, parameterName=IsMochiBody}
  )
  n14 = Multiply(A=Time(), B=ScalarParameter() {defaultValue=0.0444, parameterName=Speed - smile})
  n15 = Multiply(A=@n14) {constB=0.008}
  n16 = Add(A=ConstantBiasScale(None=Sine(None=@n14)) {scale=0.66}, B=ConstantBiasScale(None=Frac(None=@n15)) {bias=-0.5, scale=0.55})
  n17 = ScalarParameter() {defaultValue=0.12, parameterName=Smile Frequency}
  n18 = Multiply(A=@n17) {constB=2}
  n19 = ScalarParameter() {group=Mouth Pose B, parameterName=R_smile_pose}
  n20 = Add(
   A=ScalarParameter() {parameterName=SmileOverride}
   B=SphereMask(A=Saturate(None=@n16), B=NamedRerouteUsage(), Radius=LinearInterpolate(A=@n17, B=@n18, Alpha=@n19)) {attenuationRadius=0.3}
  )
  n21 = StaticBoolParameter() {defaultValue=true, parameterName=IsMochiBody}
  n22 = StaticSwitch(True=Clamp(None=Round(None=@n20)), False=@n19, Value=@n21)
  n23 = Constant()
  n24 = StaticSwitch(True=@n23, False=ScalarParameter() {group=Eye Pose C, parameterName=C_glabella_up_pose}, Value=@n21)
  n25 = StaticSwitch(True=@n23, False=ScalarParameter() {group=Eyebrow Pose B, parameterName=C_glabella_down_pose}, Value=@n21)
  n26 = Constant()
  n27 = StaticSwitch(True=@n26, False=ScalarParameter() {group=Mouth Pose C, parameterName=R_O_Phen}, Value=@n21)
  n28 = ScalarParameter() {group=Mouth Pose G, parameterName=R_Frown_pose}
  n29 = StaticSwitch(True=@n26, False=@n28, Value=@n21)
  n30 = ScalarParameter() {group=Mouth Pose F, parameterName=Jaw_open_Pose}
  n31 = StaticSwitch(True=@n26, False=@n30, Value=@n21)
  n32 = StaticSwitch(True=@n26, False=Multiply(A=@n28, B=@n30), Value=@n21)
  n33 = StaticSwitch(True=@n26, False=ScalarParameter() {group=Mouth Pose D, parameterName=R_lip_corner_narrow_pose}, Value=@n21)
  n34 = Constant()
  n35 = MaterialFunctionCall!(
   Alpha=ChannelMaskParameter(None=@n2.RGB) {maskChannel=Green, parameterName=Cloth Channel, defaultValue=<rgba 0 1 0 0>}.R
   Gmap=@n1.Gmap
   Material=SetMaterialAttributes!(
    MaterialAttributes=@n1.Result
    World Position Offset=MaterialFunctionCall!(
     Eyebrow Pose A Curve=@n4
     Eyebrow Pose B Curve=@n13
     Eyebrow Pose C Curve=@n22
     Eyebrow Pose D Curve=@n24
     Eyebrow Pose E Curve=@n25
     Eye Pose A Curve=@n4
     Eye Pose B Curve=@n13
     Eye Pose C Curve=@n22
     Mouth Pose A Curve=@n4
     Mouth Pose B Curve=@n22
     Mouth Pose C Curve=@n27
     Mouth Pose E Curve=@n29
     Mouth Pose F Curve=@n31
     Mouth Pose G Curve=@n32
     Mouth Pose D Curve=@n33
     Mouth Pose H Curve=@n34
     Mouth Pose I Curve=@n34
    ) {materialFunction=/BRCosmetics/Characters/Player/Common/MaterialFunctions/VertexColorFaces/MF_VertexColorFaces}.Result
   ) {attributeSetTypes [#1]}
  ) {materialFunction=/Game/Weapons/WeaponSkins/MaterialFunction/ClothFuzz}
  MaterialFunctionCall(
   AO=@n1.AO
   Distance Blend=@n1.AO
   Fresnel=@n1.AO
   Gmap=@n1.AO
   Pre Skinned Local Position=@n1.AO
   Material=MaterialFunctionCall!(In=MaterialFunctionCall!(Alpha=@n3.R, Material=@n35.Result) {materialFunction=/Game/Weapons/WeaponSkins/MaterialFunction/MF_Silk}.Result) {materialFunction=/Game/Weapons/WeaponSkins/MaterialFunction/MF_EmissiveFresnel}.Result
  ) {materialFunction=&BaseFN_CharacterFunction_PostFX}
  NamedRerouteDeclaration(None=@n1.Pre skinned Local Position) {name=Pre Skinned POS, nodeColor=<rgba 1 0 0.3984 1>}
  n36 = MaterialFunctionCall(Float3=Add(A=NamedRerouteUsage(), B=VectorParameter() {parameterName=Offset, defaultValue=<rgba 288 125 4.2808 1>}.RGB)) {materialFunction=&BreakOutFloat3Components}
  n37 = AppendVector(A=@n36.R, B=OneMinus(None=@n36.B))
  n38 = ScalarParameter() {defaultValue=0.0052, parameterName=Face Random Size}
  NamedRerouteDeclaration(None=TextureSample(UVs=Multiply(A=@n37, B=@n38)) {texture=/BRCosmetics/Characters/Player/Female/Medium/Bodies/F_MED_Mochi/Textures/T_MochiOFFSET}.G) {name=Random Face Value, nodeColor=<rgba 1 0 0.6797 1>}
 ]
}
