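# Physics asset for the left arm mesh of the Mochi female medium player cosmetic, setting up physics for three leaf appendages (E, F, G).
# Each leaf is a chain of a tiny master capsule and two larger dynamic capsules, linked parent to child by constraints.
# Swing limits are modest and loosen slightly toward the tip, while twist limits are zero.
# The dynamic capsules in all three chains share the same shape and leafC shape names.
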PhysicsAsset /BRCosmetics/Characters/Player/Female/Medium/Bodies/F_MED_Mochi/Meshes/F_MED_Mochi_ArmLf_Physics
bodies {
 leafE_MSTR [{shapeType=Capsule, length=0.505, radius=0.505, shapeName=leafE_MSTR_capsule, rotation=<rot 0 -90 0>}]
 leafE_dyn_1 [{shapeType=Capsule, length=2, radius=3, shapeName=leafC_dyn_1_capsule, center=<1 0 0>, rotation=<rot 90 0 0>}]
 leafE_dyn_2 [{shapeType=Capsule, length=2, radius=3, shapeName=leafC_dyn_2_capsule, center=<1 0 0>, rotation=<rot 90 0 0>}]
 leafF_MSTR [{shapeType=Capsule, length=0.505, radius=0.505, shapeName=leafF_MSTR_capsule, rotation=<rot 0 -90 0>}]
 leafF_dyn_1 [{shapeType=Capsule, length=2, radius=3, shapeName=leafC_dyn_1_capsule, center=<1 0 0>, rotation=<rot 90 0 0>}]
 leafF_dyn_2 [{shapeType=Capsule, length=2, radius=3, shapeName=leafC_dyn_2_capsule, center=<1 0 0>, rotation=<rot 90 0 0>}]
 leafG_MSTR [{shapeType=Capsule, length=0.505, radius=0.505, shapeName=leafG_MSTR_capsule, rotation=<rot 0 -90 0>}]
 leafG_dyn_1 [{shapeType=Capsule, length=2, radius=3, shapeName=leafC_dyn_1_capsule, center=<1 0 0>, rotation=<rot 90 0 0>}]
 leafG_dyn_2 [{shapeType=Capsule, length=2, radius=3, shapeName=leafC_dyn_2_capsule, center=<1 0 0>, rotation=<rot 90 0 0>}]
}
constraints [
 {bone1Name=leafF_dyn_1, bone2Name=leafF_MSTR, swing1LimitDegrees=15, swing2LimitDegrees=15, twistLimitDegrees=0}
 {bone1Name=leafF_dyn_2, bone2Name=leafF_dyn_1, swing1LimitDegrees=20, swing2LimitDegrees=20, twistLimitDegrees=0}
 {bone1Name=leafG_dyn_1, bone2Name=leafG_MSTR, swing1LimitDegrees=15, swing2LimitDegrees=15, twistLimitDegrees=0}
 {bone1Name=leafG_dyn_2, bone2Name=leafG_dyn_1, swing1LimitDegrees=20, swing2LimitDegrees=20, twistLimitDegrees=0}
 {bone1Name=leafE_dyn_1, bone2Name=leafE_MSTR, swing1LimitDegrees=15, swing2LimitDegrees=15, twistLimitDegrees=0}
 {bone1Name=leafE_dyn_2, bone2Name=leafE_dyn_1, swing1LimitDegrees=20, swing2LimitDegrees=20, twistLimitDegrees=0}
]
