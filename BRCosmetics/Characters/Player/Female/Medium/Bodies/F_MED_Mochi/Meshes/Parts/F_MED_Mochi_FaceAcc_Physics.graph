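# Physics asset for the face accessory of the Mochi medium female player cosmetic, defining bodies for a ring of petals around the head.
# Ten two-segment petal chains of capsules hang off a small petalMain capsule, with each outer segment constrained to its inner one.
# Constraints allow only narrow swing with zero twist limit, slightly looser at the outer segments.
# A large head capsule has no constraints, likely serving as a collider for the petals.
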PhysicsAsset /BRCosmetics/Characters/Player/Female/Medium/Bodies/F_MED_Mochi/Meshes/Parts/F_MED_Mochi_FaceAcc_Physics
bodies {
 Petal1 [{shapeType=Capsule, length=2, radius=3, shapeName=leafC_dyn_1_capsule, center=<2 0 0>, rotation=<rot 90 0 0>}]
 Petal10 [{shapeType=Capsule, length=2, radius=3, shapeName=leafC_dyn_1_capsule, center=<2 0 0>, rotation=<rot 90 0 0>}]
 Petal10B [{shapeType=Capsule, length=2, radius=3, shapeName=leafC_dyn_2_capsule, center=<1 0 0>, rotation=<rot 90 0 0>}]
 Petal1B [{shapeType=Capsule, length=2, radius=3, shapeName=leafC_dyn_2_capsule, center=<1 0 0>, rotation=<rot 90 0 0>}]
 Petal2 [{shapeType=Capsule, length=2, radius=3, shapeName=leafC_dyn_1_capsule, center=<2 0 0>, rotation=<rot 90 0 0>}]
 Petal2B [{shapeType=Capsule, length=2, radius=3, shapeName=leafC_dyn_2_capsule, center=<1 0 0>, rotation=<rot 90 0 0>}]
 Petal3 [{shapeType=Capsule, length=2, radius=3, shapeName=leafC_dyn_1_capsule, center=<2 0 0>, rotation=<rot 90 0 0>}]
 Petal3B [{shapeType=Capsule, length=2, radius=3, shapeName=leafC_dyn_2_capsule, center=<1 0 0>, rotation=<rot 90 0 0>}]
 Petal4 [{shapeType=Capsule, length=2, radius=3, shapeName=leafC_dyn_1_capsule, center=<2 0 0>, rotation=<rot 90 0 0>}]
 Petal4B [{shapeType=Capsule, length=2, radius=3, shapeName=leafC_dyn_2_capsule, center=<1 0 0>, rotation=<rot 90 0 0>}]
 Petal5 [{shapeType=Capsule, length=2, radius=3, shapeName=leafC_dyn_1_capsule, center=<2 0 0>, rotation=<rot 90 0 0>}]
 Petal5B [{shapeType=Capsule, length=2, radius=3, shapeName=leafC_dyn_2_capsule, center=<1 0 0>, rotation=<rot 90 0 0>}]
 Petal6 [{shapeType=Capsule, length=2, radius=3, shapeName=leafC_dyn_1_capsule, center=<2 0 0>, rotation=<rot 90 0 0>}]
 Petal6B [{shapeType=Capsule, length=2, radius=3, shapeName=leafC_dyn_2_capsule, center=<1 0 0>, rotation=<rot 90 0 0>}]
 Petal7 [{shapeType=Capsule, length=2, radius=3, shapeName=leafC_dyn_1_capsule, center=<2 0 0>, rotation=<rot 90 0 0>}]
 Petal7B [{shapeType=Capsule, length=2, radius=3, shapeName=leafC_dyn_2_capsule, center=<1 0 0>, rotation=<rot 90 0 0>}]
 Petal8 [{shapeType=Capsule, length=2, radius=3, shapeName=leafC_dyn_1_capsule, center=<2 0 0>, rotation=<rot 90 0 0>}]
 Petal8B [{shapeType=Capsule, length=2, radius=3, shapeName=leafC_dyn_2_capsule, center=<1 0 0>, rotation=<rot 90 0 0>}]
 Petal9 [{shapeType=Capsule, length=2, radius=3, shapeName=leafC_dyn_1_capsule, center=<2 0 0>, rotation=<rot 90 0 0>}]
 Petal9B [{shapeType=Capsule, length=2, radius=3, shapeName=leafC_dyn_2_capsule, center=<1 0 0>, rotation=<rot 90 0 0>}]
 head [{
   shapeType=Capsule
   length=13.1508
   radius=11.4245
   shapeName=head_capsule
   center=<6.9773 0.6758 -0.0402>
   rotation=<rot 0 -152.2997 -87.8613>
  }]
 petalMain [{shapeType=Capsule, length=0.5, radius=0.5, shapeName=petalMain_capsule, center=<1.8808 -1.289 -0.3978>, rotation=<rot 5.763 -90 17.0085>}]
}
constraints [
 {bone1Name=Petal1, bone2Name=petalMain, swing1LimitDegrees=5, swing2LimitDegrees=10, twistLimitDegrees=0}
 {bone1Name=Petal2, bone2Name=petalMain, swing1LimitDegrees=5, swing2LimitDegrees=10, twistLimitDegrees=0}
 {bone1Name=Petal3, bone2Name=petalMain, swing1LimitDegrees=5, swing2LimitDegrees=10, twistLimitDegrees=0}
 {bone1Name=Petal4, bone2Name=petalMain, swing1LimitDegrees=5, swing2LimitDegrees=10, twistLimitDegrees=0}
 {bone1Name=Petal5, bone2Name=petalMain, swing1LimitDegrees=5, swing2LimitDegrees=10, twistLimitDegrees=0}
 {bone1Name=Petal6, bone2Name=petalMain, swing1LimitDegrees=5, swing2LimitDegrees=10, twistLimitDegrees=0}
 {bone1Name=Petal7, bone2Name=petalMain, swing1LimitDegrees=5, swing2LimitDegrees=10, twistLimitDegrees=0}
 {bone1Name=Petal8, bone2Name=petalMain, swing1LimitDegrees=5, swing2LimitDegrees=10, twistLimitDegrees=0}
 {bone1Name=Petal9, bone2Name=petalMain, swing1LimitDegrees=5, swing2LimitDegrees=10, twistLimitDegrees=0}
 {bone1Name=Petal10, bone2Name=petalMain, swing1LimitDegrees=5, swing2LimitDegrees=10, twistLimitDegrees=0}
 {bone1Name=Petal1B, bone2Name=Petal1, swing1LimitDegrees=5, swing2LimitDegrees=15, twistLimitDegrees=0}
 {bone1Name=Petal2B, bone2Name=Petal2, swing1LimitDegrees=5, swing2LimitDegrees=15, twistLimitDegrees=0}
 {bone1Name=Petal3B, bone2Name=Petal3, swing1LimitDegrees=5, swing2LimitDegrees=15, twistLimitDegrees=0}
 {bone1Name=Petal4B, bone2Name=Petal4, swing1LimitDegrees=5, swing2LimitDegrees=15, twistLimitDegrees=0}
 {bone1Name=Petal5B, bone2Name=Petal5, swing1LimitDegrees=5, swing2LimitDegrees=15, twistLimitDegrees=0}
 {bone1Name=Petal6B, bone2Name=Petal6, swing1LimitDegrees=5, swing2LimitDegrees=15, twistLimitDegrees=0}
 {bone1Name=Petal7B, bone2Name=Petal7, swing1LimitDegrees=5, swing2LimitDegrees=15, twistLimitDegrees=0}
 {bone1Name=Petal8B, bone2Name=Petal8, swing1LimitDegrees=5, swing2LimitDegrees=15, twistLimitDegrees=0}
 {bone1Name=Petal9B, bone2Name=Petal9, swing1LimitDegrees=5, swing2LimitDegrees=15, twistLimitDegrees=0}
 {bone1Name=Petal10B, bone2Name=Petal10, swing1LimitDegrees=5, swing2LimitDegrees=15, twistLimitDegrees=0}
]
